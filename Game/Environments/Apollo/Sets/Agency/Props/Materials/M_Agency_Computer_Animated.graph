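# Material for the animated Agency computer prop, with spinning parts and blinking lights controlled by the Active parameter.
# A lights mask texture selects areas whose UVs rotate at two speeds, shifting the diffuse, mask and normal samples.
# Its red channel drives two looping SphereMask sweeps that flash lights at staggered times, scaled into emissive.
# The assembled material attributes feed the shared MF_Screens function.
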Material /Game/Environments/Apollo/Sets/Agency/Props/Materials/M_Agency_Computer_Animated
graph {
 properties {
  bUseMaterialAttributes=true
  pixelDepthOffsetMode=PDOM_Legacy
  previewMesh=/Game/Environments/Apollo/Sets/Agency/Props/Meshes/S_Agency_Computer_04
  usedShadingModels=Default Lit
 }
 nodes [
  n1 = TextureCoordinate()
  n2 = Rotator() {centerX=0.281, centerY=0.629, speed=1.47}
  n3 = Rotator() {centerX=0.935, centerY=0.863, speed=3.8}
  n4 = TextureSample() {texture=/Game/Environments/Apollo/Sets/Agency/Props/Textures/T_QuestComputer_LightsMask}
  n5 = LinearInterpolate(
   A=@n1
   B=LinearInterpolate(A=@n1, B=LinearInterpolate(A=@n2, B=@n3, Alpha=@n4.B), Alpha=Max(A=@n4.G, B=@n4.G))
   Alpha=ScalarParameter() {parameterName=Active}
  )
  n6 = TextureSample(UVs=@n5) {texture=/Game/Environments/Apollo/Sets/Agency/Props/Textures/T_Agency_Computer_04_D}
  n7 = TextureSample(UVs=@n5) {samplerType=SAMPLERTYPE_Masks, texture=/Game/Environments/Apollo/Sets/Agency/Props/Textures/T_Agency_Computer_04_S}
  n8 = Power(Base=@n7.A) {constExponent=3}
  n9 = ScalarParameter() {defaultValue=4, parameterName=RandomLightSpeedA}
  n10 = ScalarParameter() {defaultValue=2.8, parameterName=RandomLightSpeedB}
  n11 = Max(
   A=SphereMask(
    A=@n4.R
    B=MaterialFunctionCall!(Time, Divisor=@n9, Mod) {materialFunction=/Game/Effects/Fort_Effects/Materials/_Functions/MF_ModuloTime}.Clamped 0-1
    Radius=ScalarParameter() {defaultValue=0.05, parameterName=RandomLightDistanceA}
    Hardness=ScalarParameter() {defaultValue=0.2, parameterName=RandomLightHardnessA}
   )
   B=SphereMask(
    A=@n4.R
    B=MaterialFunctionCall(Divisor=@n10) {materialFunction=&MF_ModuloTime}.Clamped 0-1
    Radius=ScalarParameter() {defaultValue=0.08, parameterName=RandomLightDistanceB}
    Hardness=ScalarParameter() {defaultValue=0.2, parameterName=RandomLightHardnessB}
   )
  )
  n12 = Multiply(A=Multiply(A=@n6.RGB, B=Multiply(A=@n8, B=@n11)), B=ScalarParameter() {defaultValue=20, parameterName=EmissiveScale})
  n13 = Ceil(None=@n4.R)
  n14 = SetMaterialAttributes!(
   MaterialAttributes
   Base Color=@n6.RGB
   Metallic=@n7.G
   Specular=@n7.G
   Roughness=@n7.G
   Normal=TextureSample(UVs=@n5) {samplerType=SAMPLERTYPE_Normal, texture=/Game/Environments/Apollo/Sets/Agency/Props/Textures/T_Agency_Computer_04_N}.RGB
   Emissive Color=Multiply(A=Multiply(A=@n12, B=@n13), B=ScalarParameter() {parameterName=Active})
  ) {attributeSetTypes [
    #1
    #2
    #3
    #4
    #5
    #6
   ]}
  MaterialFunctionCall!(In=@n14, Emissive Mask) {materialFunction=/Game/Athena/WinterArt/Materials/MF_Screens}
 ]
}
